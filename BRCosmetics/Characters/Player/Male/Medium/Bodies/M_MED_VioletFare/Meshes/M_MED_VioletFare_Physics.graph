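# Physics asset for the VioletFare medium male player body, built around a pair of dangling loop straps.
# Capsules cover the pelvis, spine_02, legs and feet, and a wide root box forms a slab at ground level.
# Each side's strap is a constrained capsule chain anchored to the pelvis at both ends and at a middle branch, closing the loop.
# Strap joints are mostly angle-limited, while leg joints swing and twist freely.
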PhysicsAsset /BRCosmetics/Characters/Player/Male/Medium/Bodies/M_MED_VioletFare/Meshes/M_MED_VioletFare_Physics
bodies {
 calf_l [{shapeType=Capsule, length=20, radius=10, shapeName=calf_l_capsule, center=<-20 0 0>, rotation=<rot 90 83.2785 83.0429>}]
 calf_r [{shapeType=Capsule, length=20, radius=10, shapeName=calf_r_capsule, center=<20 0 0>, rotation=<rot 90 83.2785 83.0429>}]
 dyn_loop_strap_01_l [{
   shapeType=Capsule
   length=5
   radius=2.3
   shapeName=dyn_loop_strap_01_r_capsule
   center=<-4.2736 0.3513 -0.8787>
   rotation=<rot 78.7812 176.2331 -1.3712>
  }]
 dyn_loop_strap_01_r [{
   shapeType=Capsule
   length=5
   radius=2.3
   shapeName=dyn_loop_strap_01_r_capsule
   center=<4.2736 -0.3513 0.8787>
   rotation=<rot 78.7812 176.2331 178.6288>
  }]
 dyn_loop_strap_02_l [{
   shapeType=Capsule
   length=7
   radius=3.9
   shapeName=dyn_loop_strap_02_r_capsule
   center=<-2.5921 -0.8413 -0.4705>
   rotation=<rot 74.7562 -176.4804 -5.5385>
  }]
 dyn_loop_strap_02_r [{
   shapeType=Capsule
   length=7.5
   radius=3.2
   shapeName=dyn_loop_strap_02_r_capsule
   center=<2.6755 0.3064 0.4842>
   rotation=<rot 74.7562 -176.4804 174.4615>
  }]
 dyn_loop_strap_03_l [{
   shapeType=Capsule
   length=5.6728
   radius=3.652
   shapeName=dyn_loop_strap_03_r_capsule
   center=<-2.707 -1.0864 -0.1368>
   rotation=<rot -4.3571 -75.3478 -94.269>
  }]
 dyn_loop_strap_03_r [{
   shapeType=Capsule
   length=5.9728
   radius=3.352
   shapeName=dyn_loop_strap_03_r_capsule
   center=<2.8307 0.6041 0.1312>
   rotation=<rot -4.3571 -75.3478 85.731>
  }]
 dyn_loop_strap_04_l [{
   shapeType=Capsule
   length=0
   radius=3.9124
   shapeName=dyn_loop_strap_04_r_capsule
   center=<0.0032 0.3787 0.5448>
   rotation=<rot 0 85.8127 90.954>
  }]
 dyn_loop_strap_04_r [{
   shapeType=Capsule
   length=0
   radius=3.9124
   shapeName=dyn_loop_strap_04_r_capsule
   center=<-0.0032 -0.3787 -0.5448>
   rotation=<rot 0 85.8127 -89.046>
  }]
 dyn_loop_strap_05_l [{
   shapeType=Capsule
   length=6.6728
   radius=2.652
   shapeName=dyn_loop_strap_03_r_capsule
   center=<2.2003 -0.0257 -0.0723>
   rotation=<rot -5.3053 -93.5319 -86.9163>
  }]
 dyn_loop_strap_05_r [{
   shapeType=Capsule
   length=5.7728
   radius=3.552
   shapeName=dyn_loop_strap_03_r_capsule
   center=<-2.2003 0.0257 0.0723>
   rotation=<rot -5.3053 -93.5319 93.0837>
  }]
 dyn_loop_strap_06_l [{
   shapeType=Capsule
   length=3.6
   radius=2.8
   shapeName=dyn_loop_strap_02_r_capsule
   center=<2.4339 0.1515 -0.1351>
   rotation=<rot 86.5337 74.9982 -102.5648>
  }]
 dyn_loop_strap_06_r [{
   shapeType=Capsule
   length=3.9
   radius=3.1
   shapeName=dyn_loop_strap_02_r_capsule
   center=<-2.4339 -0.1515 0.1351>
   rotation=<rot 86.5337 74.9982 77.4352>
  }]
 dyn_loop_strap_07_l [{
   shapeType=Capsule
   length=6.605
   radius=1.505
   shapeName=dyn_loop_strap_07_r_capsule
   center=<4.9033 0.1207 -1.4734>
   rotation=<rot 2.998 -102.2001 112.2028>
  }]
 dyn_loop_strap_07_r [{
   shapeType=Capsule
   length=6.505
   radius=1.605
   shapeName=dyn_loop_strap_07_r_capsule
   center=<-4.9033 -0.1207 1.4734>
   rotation=<rot 2.998 -102.2001 -67.997>
  }]
 dyn_loop_strap_base_l [{
   shapeType=Capsule
   length=1.9
   radius=2
   shapeName=dyn_loop_strap_base_r_capsule
   center=<0.7181 0 0.0952>
   rotation=<rot 0 -90.0001 175.3465>
  }]
 dyn_loop_strap_base_r [{
   shapeType=Capsule
   length=0.7
   radius=3.2
   shapeName=dyn_loop_strap_base_r_capsule
   center=<-0.7181 0 -0.0952>
   rotation=<rot 0 -90.0001 -4.6535>
  }]
 dyn_loop_strap_end_l [{shapeType=Capsule, length=3.005, radius=1.505, shapeName=dyn_loop_strap_end_l_capsule, rotation=<rot 0 -90 0>}]
 dyn_loop_strap_end_r [{shapeType=Capsule, length=2.005, radius=1.505, shapeName=dyn_loop_strap_end_r_capsule, rotation=<rot 0 -90 0>}]
 dyn_loop_strap_mid_b_l [{
   shapeType=Capsule
   length=6.4
   radius=3.4124
   shapeName=dyn_loop_strap_mid_r_capsule
   center=<-0.0414 -0.447 0.6242>
   rotation=<rot -1.8679 87.5808 91.5101>
  }]
 dyn_loop_strap_mid_b_r [{
   shapeType=Capsule
   length=6.2
   radius=3.6124
   shapeName=dyn_loop_strap_mid_r_capsule
   center=<0.0249 0.0566 -0.6115>
   rotation=<rot -1.8679 87.5808 -88.4899>
  }]
 dyn_loop_strap_mid_base_l [{
   shapeType=Capsule
   length=1.4
   radius=10.905
   shapeName=dyn_loop_strap_mid_base_r_capsule
   center=<-1.079 0.2761 -0.0043>
   rotation=<rot 0 -90 180>
  }]
 dyn_loop_strap_mid_base_r [{
   shapeType=Capsule
   length=1.4
   radius=10.905
   shapeName=dyn_loop_strap_mid_base_r_capsule
   center=<1.079 -0.2761 0.0043>
   rotation=<rot 0 -90 0>
  }]
 dyn_loop_strap_mid_l [{
   shapeType=Capsule
   length=6.4
   radius=3.4124
   shapeName=dyn_loop_strap_mid_r_capsule
   center=<-0.0249 -0.0566 0.6115>
   rotation=<rot -1.8679 87.5808 91.5101>
  }]
 dyn_loop_strap_mid_r [{
   shapeType=Capsule
   length=6.2
   radius=3.6124
   shapeName=dyn_loop_strap_mid_r_capsule
   center=<0.0249 0.0566 -0.6115>
   rotation=<rot -1.8679 87.5808 -88.4899>
  }]
 dyn_loop_strap_mid_root_l [{shapeType=Capsule, length=0.505, radius=0.505, shapeName=dyn_loop_strap_mid_root_r_capsule, rotation=<rot 0 -90 180>}]
 dyn_loop_strap_mid_root_r [{shapeType=Capsule, length=0.505, radius=0.505, shapeName=dyn_loop_strap_mid_root_r_capsule, rotation=<rot 0 -90 0>}]
 foot_l [{shapeType=Capsule, length=10, radius=10, shapeName=foot_l_capsule, center=<-5 5 0>, rotation=<rot 90 -3.6023 86.1621>}]
 foot_r [{shapeType=Capsule, length=10, radius=10, shapeName=foot_r_capsule, center=<5 -5 0>, rotation=<rot 90 -3.6023 86.3577>}]
 pelvis [
  {shapeType=Capsule, length=1.5, radius=13.9601, shapeName=pelvis_capsule, center=<-8.2351 1.1918 0.1047>, rotation=<rot 0 1.1886 -179.4437>}
  {shapeType=Capsule, length=8.1, radius=15, shapeName=pelvis_capsule, center=<0 1.0818 0>}
 ]
 root [{shapeType=Box, extentX=153.61, extentY=166.21, extentZ=35.01, shapeName=root_box, center=<0 0 -17.1594>, rotation=<rot 0 -90 0>}]
 spine_02 [{shapeType=Capsule, length=8.9841, radius=13.2016, center=<3.8205 0.4425 -0.1051>, rotation=<rot 4.3731 -89.9997 -0.01>}]
 thigh_l [{shapeType=Capsule, length=29.6, radius=9.3, shapeName=thigh_l_capsule, center=<-16.4003 -0.864 0.9702>, rotation=<rot 90 90 89.96>}]
 thigh_r [{shapeType=Capsule, length=29.6, radius=9.3, shapeName=thigh_l_capsule, center=<16.4003 0.864 -0.9702>, rotation=<rot 90 165.1635 -14.8765>}]
}
constraints [
 {bone1Name=spine_02, bone2Name=pelvis, swing1LimitDegrees=45, swing2LimitDegrees=45, twistLimitDegrees=45}
 {
  bone1Name=calf_l
  bone2Name=thigh_l
  swing1LimitDegrees=45
  swing1Motion=Free
  swing2LimitDegrees=45
  swing2Motion=Free
  twistLimitDegrees=45
  twistMotion=Free
 }
 {
  bone1Name=foot_l
  bone2Name=calf_l
  swing1LimitDegrees=45
  swing1Motion=Free
  swing2LimitDegrees=45
  swing2Motion=Free
  twistLimitDegrees=45
  twistMotion=Free
 }
 {
  bone1Name=calf_r
  bone2Name=thigh_r
  swing1LimitDegrees=45
  swing1Motion=Free
  swing2LimitDegrees=45
  swing2Motion=Free
  twistLimitDegrees=45
  twistMotion=Free
 }
 {
  bone1Name=foot_r
  bone2Name=calf_r
  swing1LimitDegrees=45
  swing1Motion=Free
  swing2LimitDegrees=45
  swing2Motion=Free
  twistLimitDegrees=45
  twistMotion=Free
 }
 {bone1Name=dyn_loop_strap_base_r, bone2Name=pelvis, swing1LimitDegrees=60, swing2LimitDegrees=60, twistLimitDegrees=5}
 {bone1Name=dyn_loop_strap_01_r, bone2Name=dyn_loop_strap_base_r, swing1LimitDegrees=45, swing2LimitDegrees=45, twistLimitDegrees=5}
 {bone1Name=dyn_loop_strap_02_r, bone2Name=dyn_loop_strap_01_r, swing1LimitDegrees=45, swing2LimitDegrees=45, twistLimitDegrees=5}
 {
  bone1Name=thigh_r
  bone2Name=pelvis
  swing1LimitDegrees=45
  swing1Motion=Free
  swing2LimitDegrees=45
  swing2Motion=Free
  twistLimitDegrees=45
  twistMotion=Free
 }
 {bone1Name=dyn_loop_strap_end_r, bone2Name=pelvis, swing1LimitDegrees=60, swing2LimitDegrees=60, twistLimitDegrees=5}
 {
  bone1Name=dyn_loop_strap_end_l
  bone2Name=pelvis
  swing1LimitDegrees=45
  swing1Motion=Free
  swing2LimitDegrees=45
  swing2Motion=Free
  twistLimitDegrees=45
  twistMotion=Free
 }
 {
  bone1Name=pelvis
  bone2Name=root
  swing1LimitDegrees=45
  swing1Motion=Free
  swing2LimitDegrees=45
  swing2Motion=Free
  twistLimitDegrees=45
  twistMotion=Free
 }
 {bone1Name=dyn_loop_strap_04_r, bone2Name=dyn_loop_strap_mid_base_r, swing1LimitDegrees=15, swing2LimitDegrees=15, twistLimitDegrees=15}
 {bone1Name=dyn_loop_strap_03_r, bone2Name=dyn_loop_strap_02_r, swing1LimitDegrees=45, swing2LimitDegrees=45, twistLimitDegrees=5}
 {bone1Name=dyn_loop_strap_end_r, bone2Name=dyn_loop_strap_07_r, swing1LimitDegrees=45, swing2LimitDegrees=45, twistLimitDegrees=15}
 {bone1Name=dyn_loop_strap_06_r, bone2Name=dyn_loop_strap_07_r, swing1LimitDegrees=45, swing2LimitDegrees=45, twistLimitDegrees=15}
 {bone1Name=dyn_loop_strap_05_r, bone2Name=dyn_loop_strap_06_r, swing1LimitDegrees=45, swing2LimitDegrees=45, twistLimitDegrees=15}
 {bone1Name=dyn_loop_strap_base_l, bone2Name=pelvis, swing1LimitDegrees=35, swing2LimitDegrees=35, twistLimitDegrees=15}
 {
  bone1Name=dyn_loop_strap_mid_base_r
  bone2Name=dyn_loop_strap_mid_root_r
  swing1LimitDegrees=70
  swing2LimitDegrees=60
  twistLimitDegrees=15
  twistMotion=Locked
 }
 {
  bone1Name=dyn_loop_strap_mid_root_r
  bone2Name=pelvis
  swing1LimitDegrees=45
  swing1Motion=Free
  swing2LimitDegrees=45
  swing2Motion=Free
  twistLimitDegrees=45
  twistMotion=Free
 }
 {bone1Name=dyn_loop_strap_mid_r, bone2Name=dyn_loop_strap_03_r, swing1LimitDegrees=45, swing2LimitDegrees=45, twistLimitDegrees=5}
 {bone1Name=dyn_loop_strap_mid_r, bone2Name=dyn_loop_strap_04_r, swing1LimitDegrees=5, swing2LimitDegrees=5, twistLimitDegrees=5}
 {bone1Name=dyn_loop_strap_mid_b_r, bone2Name=dyn_loop_strap_05_r, swing1LimitDegrees=5, swing2LimitDegrees=5, twistLimitDegrees=5}
 {bone1Name=dyn_loop_strap_04_r, bone2Name=dyn_loop_strap_mid_b_r, swing1LimitDegrees=45, swing2LimitDegrees=45, twistLimitDegrees=15}
 {bone1Name=dyn_loop_strap_01_l, bone2Name=dyn_loop_strap_base_l, swing1LimitDegrees=45, swing2LimitDegrees=45, twistLimitDegrees=5}
 {bone1Name=dyn_loop_strap_02_l, bone2Name=dyn_loop_strap_01_l, swing1LimitDegrees=45, swing2LimitDegrees=45, twistLimitDegrees=5}
 {
  bone1Name=thigh_l
  bone2Name=pelvis
  swing1LimitDegrees=45
  swing1Motion=Free
  swing2LimitDegrees=45
  swing2Motion=Free
  twistLimitDegrees=45
  twistMotion=Free
 }
 {
  bone1Name=dyn_loop_strap_mid_b_l
  bone2Name=dyn_loop_strap_05_l
  swing1LimitDegrees=45
  swing1Motion=Locked
  swing2LimitDegrees=45
  swing2Motion=Locked
  twistLimitDegrees=15
 }
 {bone1Name=dyn_loop_strap_05_l, bone2Name=dyn_loop_strap_06_l, swing1LimitDegrees=45, swing2LimitDegrees=45, twistLimitDegrees=15}
 {bone1Name=dyn_loop_strap_06_l, bone2Name=dyn_loop_strap_07_l, swing1LimitDegrees=45, swing2LimitDegrees=45, twistLimitDegrees=15}
 {bone1Name=dyn_loop_strap_07_l, bone2Name=dyn_loop_strap_end_l, swing1LimitDegrees=45, swing2LimitDegrees=45, twistLimitDegrees=15}
 {bone1Name=dyn_loop_strap_mid_l, bone2Name=dyn_loop_strap_03_l, swing1LimitDegrees=45, swing2LimitDegrees=45, twistLimitDegrees=5}
 {bone1Name=dyn_loop_strap_03_l, bone2Name=dyn_loop_strap_02_l, swing1LimitDegrees=45, swing2LimitDegrees=45, twistLimitDegrees=5}
 {bone1Name=dyn_loop_strap_04_l, bone2Name=dyn_loop_strap_mid_base_l, swing1LimitDegrees=15, swing2LimitDegrees=15, twistLimitDegrees=15}
 {
  bone1Name=dyn_loop_strap_mid_base_l
  bone2Name=dyn_loop_strap_mid_root_l
  swing1LimitDegrees=70
  swing2LimitDegrees=60
  twistLimitDegrees=15
  twistMotion=Locked
 }
 {
  bone1Name=dyn_loop_strap_mid_root_l
  bone2Name=pelvis
  swing1LimitDegrees=45
  swing1Motion=Free
  swing2LimitDegrees=45
  swing2Motion=Free
  twistLimitDegrees=45
  twistMotion=Free
 }
 {bone1Name=dyn_loop_strap_04_l, bone2Name=dyn_loop_strap_mid_l, swing1LimitDegrees=5, swing2LimitDegrees=5, twistLimitDegrees=5}
 {bone1Name=dyn_loop_strap_04_l, bone2Name=dyn_loop_strap_mid_b_l, swing1LimitDegrees=5, swing2LimitDegrees=5, twistLimitDegrees=5}
]
