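# Physics asset for the VioletFare medium male player body cosmetic, simulating two dangling back straps.
# Capsule-chained strap bones hang from base bodies on the pelvis, seven segments left and five right, with boxes on 06_l and 05_r.
# Limited swing and tight twist constrain the links, while the base and leg joints have free angular motion.
# Pelvis, spine_02, leg and foot capsules plus a wide, floor-like root box provide collision shapes around the straps.
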PhysicsAsset /BRCosmetics/Characters/Player/Male/Medium/Bodies/M_MED_VioletFare/Meshes/M_MED_VioletFare_Physics_BackStraps
bodies {
 calf_l [{shapeType=Capsule, length=27.2, radius=7.2, shapeName=calf_r_capsule, center=<-20.5753 -0.0024 0>, rotation=<rot 90 160.5307 -19.7049>}]
 calf_r [{shapeType=Capsule, length=27.2, radius=7.2, shapeName=calf_r_capsule, center=<20.5753 0.0024 0>, rotation=<rot 90 170.1389 169.9033>}]
 dyn_strap_01_l [{shapeType=Capsule, length=9.7, radius=2, shapeName=dyn_strap_01_l_capsule, center=<-6.6548 0.0176 0>, rotation=<rot 0 91.8462 90>}]
 dyn_strap_01_r [{shapeType=Capsule, length=9.5, radius=2, shapeName=dyn_strap_01_r_capsule, center=<7.0282 -0.0061 0>, rotation=<rot 0 92.286 90>}]
 dyn_strap_02_l [{shapeType=Capsule, length=7.7, radius=2, shapeName=dyn_strap_02_l_capsule, center=<-5.6284 0 0>, rotation=<rot 0 90 89.8474>}]
 dyn_strap_02_r [{shapeType=Capsule, length=7.6, radius=2, shapeName=dyn_strap_02_r_capsule, center=<5.62 0 0>, rotation=<rot 0 90 90>}]
 dyn_strap_03_l [{shapeType=Capsule, length=7.6, radius=2, shapeName=dyn_strap_03_l_capsule, center=<-5.8248 0 0>, rotation=<rot 0 90 89.8474>}]
 dyn_strap_03_r [{shapeType=Capsule, length=7.6, radius=2, shapeName=dyn_strap_03_r_capsule, center=<5.8969 0 0>, rotation=<rot 0 90 89.8474>}]
 dyn_strap_04_l [{shapeType=Capsule, length=6.7, radius=2, shapeName=dyn_strap_04_l_capsule, center=<-4.9131 0 0>, rotation=<rot 0 90 90>}]
 dyn_strap_04_r [{shapeType=Capsule, length=6.6, radius=2, shapeName=dyn_strap_04_r_capsule, center=<4.8999 0 0>, rotation=<rot 0 90 90>}]
 dyn_strap_05_l [{shapeType=Capsule, length=6.9, radius=2, shapeName=dyn_strap_05_l_capsule, center=<-4.8516 0 0>, rotation=<rot 0 90 90>}]
 dyn_strap_05_r [{
   shapeType=Box
   extentX=1.6931
   extentY=6.605
   extentZ=13.4483
   shapeName=dyn_strap_05_r_box
   center=<4.522 0.0795 -0.7501>
   rotation=<rot 0 -90.01 -34.8502>
  }]
 dyn_strap_06_l [{
   shapeType=Box
   extentX=2.0523
   extentY=6.2259
   extentZ=12.6415
   shapeName=dyn_strap_06_l_box
   center=<-3.8271 -0.0481 0.0433>
   rotation=<rot -0.2039 -89.9999 -9.7739>
  }]
 dyn_strap_07_l [{shapeType=Capsule, length=5, radius=2.3, shapeName=dyn_strap_06_l_capsule, center=<-3.8975 -0.9276 0.265>, rotation=<rot 0 90 87.1284>}]
 dyn_strap_base_l [{shapeType=Capsule, length=2, radius=2, shapeName=dyn_strap_base_l_capsule}]
 dyn_strap_base_r [{shapeType=Capsule, length=2, radius=2, shapeName=dyn_strap_base_r_capsule}]
 foot_l [{shapeType=Capsule, length=12.4, radius=7.6, shapeName=foot_r_capsule, center=<-4.9996 4.4388 0>, rotation=<rot 90 180 89.96>}]
 foot_r [{shapeType=Capsule, length=12.4, radius=7.6, shapeName=foot_r_capsule, center=<4.9996 -4.4388 0>, rotation=<rot 90 180 -90.04>}]
 pelvis [{shapeType=Capsule, length=9.6683, radius=15.4601, shapeName=pelvis_capsule, center=<0.1779 2.1305 0.0973>, rotation=<rot 0 1.1886 -179.4437>}]
 root [{shapeType=Box, extentX=153.61, extentY=166.21, extentZ=35.01, shapeName=root_box, center=<0 0 -17.1594>, rotation=<rot 0 -90 -0.1526>}]
 spine_02 [{shapeType=Capsule, length=8.9841, radius=13.2016, center=<3.8205 0.4425 -0.1051>, rotation=<rot 4.3731 -89.9997 -0.01>}]
 thigh_l [{shapeType=Capsule, length=28.9, radius=10, shapeName=thigh_l_capsule, center=<-16.4003 -0.864 1.4826>, rotation=<rot 90 90 89.96>}]
 thigh_r [{shapeType=Capsule, length=28.9, radius=10, shapeName=thigh_l_capsule, center=<16.4003 0.864 -1.4826>, rotation=<rot 90 165.1635 -14.8765>}]
}
constraints [
 {bone1Name=spine_02, bone2Name=pelvis, swing1LimitDegrees=45, swing2LimitDegrees=45, twistLimitDegrees=45}
 {
  bone1Name=dyn_strap_base_l
  bone2Name=pelvis
  swing1LimitDegrees=45
  swing1Motion=Free
  swing2LimitDegrees=45
  swing2Motion=Free
  twistLimitDegrees=45
  twistMotion=Free
 }
 {bone1Name=dyn_strap_01_l, bone2Name=dyn_strap_base_l, swing1LimitDegrees=60, swing2LimitDegrees=25, twistLimitDegrees=3}
 {bone1Name=dyn_strap_02_l, bone2Name=dyn_strap_01_l, swing1LimitDegrees=45, swing2LimitDegrees=45, twistLimitDegrees=10}
 {bone1Name=dyn_strap_03_l, bone2Name=dyn_strap_02_l, swing1LimitDegrees=45, swing2LimitDegrees=45, twistLimitDegrees=10}
 {bone1Name=dyn_strap_04_l, bone2Name=dyn_strap_03_l, swing1LimitDegrees=45, swing2LimitDegrees=45, twistLimitDegrees=10}
 {bone1Name=dyn_strap_05_l, bone2Name=dyn_strap_04_l, swing1LimitDegrees=45, swing2LimitDegrees=45, twistLimitDegrees=10}
 {bone1Name=dyn_strap_06_l, bone2Name=dyn_strap_05_l, swing1LimitDegrees=45, swing2LimitDegrees=45, twistLimitDegrees=10}
 {
  bone1Name=dyn_strap_base_r
  bone2Name=pelvis
  swing1LimitDegrees=45
  swing1Motion=Free
  swing2LimitDegrees=45
  swing2Motion=Free
  twistLimitDegrees=45
  twistMotion=Free
 }
 {bone1Name=dyn_strap_01_r, bone2Name=dyn_strap_base_r, swing1LimitDegrees=60, swing2LimitDegrees=25, twistLimitDegrees=3}
 {bone1Name=dyn_strap_02_r, bone2Name=dyn_strap_01_r, swing1LimitDegrees=45, swing2LimitDegrees=45, twistLimitDegrees=10}
 {bone1Name=dyn_strap_03_r, bone2Name=dyn_strap_02_r, swing1LimitDegrees=45, swing2LimitDegrees=45, twistLimitDegrees=10}
 {bone1Name=dyn_strap_04_r, bone2Name=dyn_strap_03_r, swing1LimitDegrees=45, swing2LimitDegrees=45, twistLimitDegrees=10}
 {bone1Name=dyn_strap_05_r, bone2Name=dyn_strap_04_r, swing1LimitDegrees=45, swing2LimitDegrees=45, twistLimitDegrees=10}
 {
  bone1Name=foot_r
  bone2Name=calf_r
  swing1LimitDegrees=45
  swing1Motion=Free
  swing2LimitDegrees=45
  swing2Motion=Free
  twistLimitDegrees=45
  twistMotion=Free
 }
 {
  bone1Name=thigh_r
  bone2Name=pelvis
  swing1LimitDegrees=45
  swing1Motion=Free
  swing2LimitDegrees=45
  swing2Motion=Free
  twistLimitDegrees=45
  twistMotion=Free
 }
 {
  bone1Name=pelvis
  bone2Name=root
  swing1LimitDegrees=45
  swing1Motion=Free
  swing2LimitDegrees=45
  swing2Motion=Free
  twistLimitDegrees=45
  twistMotion=Free
 }
 {bone1Name=dyn_strap_07_l, bone2Name=dyn_strap_06_l, swing1LimitDegrees=45, swing2LimitDegrees=45, twistLimitDegrees=10}
 {
  bone1Name=foot_l
  bone2Name=calf_l
  swing1LimitDegrees=45
  swing1Motion=Free
  swing2LimitDegrees=45
  swing2Motion=Free
  twistLimitDegrees=45
  twistMotion=Free
 }
 {
  bone1Name=calf_l
  bone2Name=thigh_l
  swing1LimitDegrees=45
  swing1Motion=Free
  swing2LimitDegrees=45
  swing2Motion=Free
  twistLimitDegrees=45
  twistMotion=Free
 }
 {
  bone1Name=thigh_l
  bone2Name=pelvis
  swing1LimitDegrees=45
  swing1Motion=Free
  swing2LimitDegrees=45
  swing2Motion=Free
  twistLimitDegrees=45
  twistMotion=Free
 }
 {
  bone1Name=calf_r
  bone2Name=thigh_r
  swing1LimitDegrees=45
  swing1Motion=Free
  swing2LimitDegrees=45
  swing2Motion=Free
  twistLimitDegrees=45
  twistMotion=Free
 }
]
