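# Post-process style Anim Blueprint for the M_MED_VioletFare player body cosmetic, adding secondary physics to a pose copied from the parent mesh.
# A Control Rig feeds chained RigidBody nodes: two back-strap sims with different tuning, toggled oppositely by whether a cape backbling is worn, then a pelvis-based sim.
# The update event detects the cape, swaps gravity override params for cape skydiving, computes Gravity, and sets LOD and rigid-body-off flags.
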AnimBlueprint /BRCosmetics/Characters/Player/Male/Medium/Bodies/M_MED_VioletFare/Meshes/M_MED_VioletFare_AnimBP
target_skeleton=&/M_MED_VioletFare_Skeleton
graphs [
 AnimGraph {nodes [
   n1 = Control Rig(Source=Copy Pose From Mesh() {bUseAttachedParent=true}) {
    alphaBoolBlend {blendOption=Linear}
    controlRigAssetReference {blueprintRigClass=&/M_MED_VioletFare_CtrlRig.M_MED_VioletFare_CtrlRig_C}
   }
   n2 = Get Wear_cape() -> (Wear_cape:bool)
   n3 = RigidBody!(ComponentPose=Local To Component(LocalPose=@n1), bAlphaBoolEnabled:bool=@n2, OverrideWorldGravity:Vector="0,0,-980", ExternalForce) -> (Pose) {
    alphaInputType=Bool
    bClampLinearTranslationLimitToRefPose=true
    bOverrideWorldGravity=true
    overridePhysicsAsset=&/M_MED_VioletFare_Physics_BackStraps
    simulationSpace=BaseBoneSpace
    alphaBoolBlend {blendOption=Linear}
    baseBoneRef {boneName=dyn_strap_root}
    componentAppliedLinearAccClamp=<5000 5000 5000>
    overrideWorldGravity {z=-980}
    simSpaceSettings {maxAngularAcceleration=300, maxAngularVelocity=2, maxLinearAcceleration=300, maxLinearVelocity=80, velocityScaleZ=0, worldAlpha=0.4}
   }
   n4 = RigidBody(ComponentPose=@n3, bAlphaBoolEnabled=NOT Boolean(A=@n2), OverrideWorldGravity="0,0,-980") {
    alphaInputType=Bool
    bClampLinearTranslationLimitToRefPose=true
    bOverrideWorldGravity=true
    overridePhysicsAsset=&M_MED_VioletFare_Physics_BackStraps
    simulationSpace=BaseBoneSpace
    alphaBoolBlend {blendOption=Linear}
    baseBoneRef {boneName=dyn_strap_root}
    overrideWorldGravity {z=-980}
    simSpaceSettings {
     maxAngularAcceleration=500
     maxAngularVelocity=3
     maxLinearAcceleration=450
     maxLinearVelocity=200
     velocityScaleZ=0
     worldAlpha=1
     externalLinearDragV=<0.1 0.1>
    }
   }
   n5 = RigidBody!(ComponentPose=@n4, Alpha:float, OverrideWorldGravity="0,0,-980", ExternalForce) -> (Pose) {
    bClampLinearTranslationLimitToRefPose=true
    bOverrideWorldGravity=true
    overridePhysicsAsset=&/M_MED_VioletFare_Physics
    simulationSpace=BaseBoneSpace
    alphaBoolBlend {blendOption=Linear}
    baseBoneRef {boneName=pelvis}
    overrideWorldGravity {z=-980}
    simSpaceSettings {
     maxAngularAcceleration=500
     maxAngularVelocity=3
     maxLinearAcceleration=450
     maxLinearVelocity=35
     velocityScaleZ=0
     worldAlpha=0.5
     externalLinearDragV=<0.1 0.1>
    }
   }
   Output Pose(Result=Component To Local(ComponentPose=@n5)) {layerGroup=None}
  ]}
 EventGraph {nodes [
   Event BlueprintUpdateAnimation()
   > n1 = Sequence!(execute) -> (then_0, then_1, then_2:Exec, then_3:Exec)
   n2 = Branch(execute=@n1.then_1, Condition=Get bIsCrouching())
   > Set suspenders_component_linear(suspenders_component_linear:Vector="0, 0, 0") -> (Output_Get:Vector)
   Set suspenders_component_linear(execute=@n2.else, suspenders_component_linear="0, 0, 0")
   n3 = Sequence(execute=@n1.then_2)
   n4 = Get CurrentLOD()
   > Set LOD3(LOD3:bool="integer >= integer"(A=@n4, B=3)) -> (Output_Get:bool)
   Set LOD2(execute=@n3.then_1, LOD2:bool="integer >= integer"(A=@n4, B=2)) -> (Output_Get:bool)
   Set LOD1(execute=@n3.then_2, LOD1:bool="integer >= integer"(A=@n4, B=1)) -> (Output_Get:bool)
   Set RBAN_off(execute=@n3.then_3, RBAN_off:bool=NOT Boolean(A=Get bRigidBodyEnabled())) -> (Output_Get:bool)
   n5 = Set Wear_cape(execute=@n1.then_3, Wear_cape:bool=CheckForBackbling(self=Self-Reference()).bIsBackblingACape) -> (Output_Get:bool)
   > Branch(Condition=@n5)
   n6 = Gravity Override State Function DEPRECATED(
    execute=@n1.then_0
    CharacterPartAnimInstance=Self-Reference()
    GameTimeInSeconds=GetGameTimeInSeconds()
    GravityOverrideParams=Get Gravity Override Params()
   )
   > Set Gravity(Gravity:Vector=@n6) -> (Output_Get:Vector)
   > n7 = Branch(Condition=Get Wear_cape())
   > Set Gravity Override Params(Gravity Override Params:Gravity Override Params Struct Structure=Get Gravity Override Params_Skydive_Cape() -> (Gravity Override Params_Skydive_Cape:Gravity Override Params Struct Structure)) -> (Output_Get:Gravity Override Params Struct Structure)
   Set Gravity Override Params(
    execute=@n7.else
    Gravity Override Params=Get Gravity Override Params_Default() -> (Gravity Override Params_Default:Gravity Override Params Struct Structure)
   )
   Event BlueprintInitializeAnimation()
   > Set SkelMesh(SkelMesh=GetOwnerSkeletalMeshComponent())
   n8 = TryGetPawnOwner()
   > Is Valid(InputObject=@n8)
   > n9 = Cast To FortPlayerPawnAthena(Object=@n8)
   > Set PlayerPawn(PlayerPawn:Fort Player Pawn Athena Object Reference=@n9) -> (Output_Get:Fort Player Pawn Athena Object Reference)
   Set Alpha Rigid Body(Alpha Rigid Body:double=0) -> (Output_Get:double)
  ]}
]
variables [
 Alpha Rigid Body {type=number}
 Gravity {type=Vector}
 Gravity Override Params {type=GravityOverrideParamsStruct}
 Gravity Override Params_Default {
  type=GravityOverrideParamsStruct
  default {dBNO {jointName=root, gravityOverride=<0 0 -980>}, dBNO_Carried {jointName=root, gravityOverride=<0 0 -980>}}
 }
 Gravity Override Params_Skydive_Cape {
  type=GravityOverrideParamsStruct
  default {
   dBNO {jointName=root, gravityOverride=<0 0 -980>}
   dBNO_Carried {jointName=root, gravityOverride=<0 0 -980>}
   skydive_FloatingDown {gravityOverride=<0 -700 500>, windAmplitude=<100 -250 200>, windOffset=<0 -1000 -500>}
   skydive_GoingUp {gravityOverride=<0 -700 500>, windAmplitude=<100 -250 200>, windOffset=<0 -1000 -500>}
   skydive_Parachuting {windAmplitude=<100 -250 200>, windOffset=<0 0 -250>}
  }
 }
 LOD1 {type=boolean}
 LOD2 {type=boolean}
 LOD3 {type=boolean}
 PlayerPawn {type=FortPlayerPawnAthena}
 RBAN Params {
  type=RBANParamsStruct
  default {
   dBNO {simSpaceSettings {
     maxAngularAcceleration=1000
     maxAngularVelocity=3
     maxLinearAcceleration=300
     maxLinearVelocity=700
     velocityScaleZ=0
     externalLinearDragV=<0 0.5 0>
    }}
   dBNO_Carried {simSpaceSettings {
     maxAngularAcceleration=1000
     maxAngularVelocity=3
     maxLinearAcceleration=300
     maxLinearVelocity=700
     velocityScaleZ=0
     externalLinearDragV=<0 0.5 0>
    }}
   emoteOrMelee {simSpaceSettings {
     maxAngularAcceleration=1000
     maxAngularVelocity=3
     maxLinearAcceleration=300
     maxLinearVelocity=700
     velocityScaleZ=0
     externalLinearDragV=<0 0.5 0>
    }}
   front_End {simSpaceSettings {
     maxAngularAcceleration=1000
     maxAngularVelocity=3
     maxLinearAcceleration=300
     maxLinearVelocity=700
     velocityScaleZ=0
     externalLinearDragV=<0 0.5 0>
    }}
   onGround_Crouch_Moving {simSpaceSettings {
     maxAngularAcceleration=1000
     maxAngularVelocity=3
     maxLinearAcceleration=300
     maxLinearVelocity=700
     velocityScaleZ=0
     externalLinearDragV=<0 0.5 0>
    }}
   onGround_Falling {simSpaceSettings {
     maxAngularAcceleration=1000
     maxAngularVelocity=3
     maxLinearAcceleration=300
     maxLinearVelocity=700
     velocityScaleZ=0
     externalLinearDragV=<0 0.5 0>
    }}
   onGround_Moving {simSpaceSettings {
     maxAngularAcceleration=1000
     maxAngularVelocity=3
     maxLinearAcceleration=300
     maxLinearVelocity=700
     velocityScaleZ=0
     externalLinearDragV=<0 0.5 0>
    }}
   onGround_Standing {simSpaceSettings {
     maxAngularAcceleration=1000
     maxAngularVelocity=3
     maxLinearAcceleration=300
     maxLinearVelocity=700
     velocityScaleZ=0
     externalLinearDragV=<0 0.5 0>
    }}
   rail_Grinding {simSpaceSettings {
     maxAngularAcceleration=1000
     maxAngularVelocity=3
     maxLinearAcceleration=300
     maxLinearVelocity=700
     velocityScaleZ=0
     worldAlpha=0.1
     externalLinearDragV=<0 0.5 0>
    }}
   riding_Moving {simSpaceSettings {
     maxAngularAcceleration=1000
     maxAngularVelocity=3
     maxLinearAcceleration=300
     maxLinearVelocity=700
     velocityScaleZ=0
     externalLinearDragV=<0 0.5 0>
    }}
   riding_Moving_Targeting {simSpaceSettings {
     maxAngularAcceleration=1000
     maxAngularVelocity=3
     maxLinearAcceleration=300
     maxLinearVelocity=700
     velocityScaleZ=0
     externalLinearDragV=<0 0.5 0>
    }}
   riding_Standing {simSpaceSettings {
     maxAngularAcceleration=1000
     maxAngularVelocity=3
     maxLinearAcceleration=300
     maxLinearVelocity=700
     velocityScaleZ=0
     externalLinearDragV=<0 0.5 0>
    }}
   skydive_DivingDown {simSpaceSettings {
     maxAngularAcceleration=1000
     maxAngularVelocity=3
     maxLinearAcceleration=300
     maxLinearVelocity=700
     velocityScaleZ=0
     externalLinearDragV=<0 0.5 0>
    }}
   skydive_FloatingDown {
    componentLinearAccScale=<0 0.5 0>
    componentLinearVelScale=<0 0.5 0>
    simSpaceSettings {
     maxAngularAcceleration=1000
     maxAngularVelocity=3
     maxLinearAcceleration=300
     maxLinearVelocity=700
     velocityScaleZ=0
     externalLinearDragV=<0 0.5 0>
    }
   }
   skydive_GoingUp {simSpaceSettings {
     maxAngularAcceleration=1000
     maxAngularVelocity=3
     maxLinearAcceleration=300
     maxLinearVelocity=700
     velocityScaleZ=0
     externalLinearDragV=<0 0.5 0>
    }}
   skydive_Parachuting {
    componentLinearAccScale=<0 0.1 0.3>
    componentLinearVelScale=<0 0.1 0.3>
    simSpaceSettings {
     maxAngularAcceleration=1000
     maxAngularVelocity=3
     maxLinearAcceleration=300
     maxLinearVelocity=700
     velocityScaleZ=0
     externalLinearDragV=<0 0.5 0>
    }
   }
   skydive_Surfing {simSpaceSettings {
     maxAngularAcceleration=1000
     maxAngularVelocity=3
     maxLinearAcceleration=300
     maxLinearVelocity=700
     velocityScaleZ=0
     externalLinearDragV=<0 0.5 0>
    }}
   swimming {simSpaceSettings {
     maxAngularAcceleration=1000
     maxAngularVelocity=3
     maxLinearAcceleration=300
     maxLinearVelocity=700
     velocityScaleZ=0
     externalLinearDragV=<0 0.5 0>
    }}
   tactical_Sprint {simSpaceSettings {
     maxAngularAcceleration=1000
     maxAngularVelocity=3
     maxLinearAcceleration=300
     maxLinearVelocity=700
     velocityScaleZ=0
     externalLinearDragV=<0 0.5 0>
    }}
   ziplining {simSpaceSettings {
     maxAngularAcceleration=1000
     maxAngularVelocity=3
     maxLinearAcceleration=300
     maxLinearVelocity=700
     velocityScaleZ=0
     externalLinearDragV=<0 0.5 0>
    }}
  }
 }
 RBAN_off {type=boolean}
 SkelMesh {type=SkeletalMeshComponent}
 Skydive Enabled {type=boolean}
 Skydive with Cape {type=boolean}
 suspenders_component_linear {type=Vector}
 Wear_cape {type=boolean}
]
