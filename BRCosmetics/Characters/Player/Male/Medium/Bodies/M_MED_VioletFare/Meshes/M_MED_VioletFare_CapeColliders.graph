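# Physics asset of cape colliders for the VioletFare medium male player body, covering the torso, head and two dyn_strap bone chains.
# Capsules and spheres wrap pelvis, spine, clavicles, neck and head, while limb bodies have no shapes.
# Pelvis-anchored sphere chains, longer on the left, use limited swing and tighter twist; other joints rotate freely.
# A huge, near-spherical root capsule centered far below the origin apparently approximates the ground.
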PhysicsAsset /BRCosmetics/Characters/Player/Male/Medium/Bodies/M_MED_VioletFare/Meshes/M_MED_VioletFare_CapeColliders
bodies {
 calf_l []
 calf_r []
 clavicle_l [{shapeType=Capsule, length=9.6192, radius=9.8871, center=<11.7113 -1.6766 2.5574>, rotation=<rot 81.02 -168.8572 6.9216>}]
 clavicle_r [{shapeType=Capsule, length=9.6192, radius=9.8871, center=<-11.7113 1.6766 -2.5574>, rotation=<rot 81.02 -168.8572 -173.0784>}]
 dyn_strap_01_l [{shapeType=Sphere, radius=7.1535, shapeName=dyn_strap_01_l_sphere, center=<-8.1189 -2.8019 0.1451>}]
 dyn_strap_01_r [{shapeType=Sphere, radius=7.2535, shapeName=dyn_strap_01_l_sphere, center=<8.1189 2.8019 -0.1451>}]
 dyn_strap_02_l [{shapeType=Sphere, radius=8.105, shapeName=dyn_strap_02_l_sphere, center=<-5.2421 -3.6537 -0.0041>}]
 dyn_strap_02_r [{shapeType=Sphere, radius=8.105, shapeName=dyn_strap_02_r_sphere, center=<5.241 3.6569 0.0021>}]
 dyn_strap_03_l [{shapeType=Sphere, radius=11.005, shapeName=dyn_strap_03_l_sphere, center=<-6.0253 -8.2254 0.0022>}]
 dyn_strap_03_r [{shapeType=Sphere, radius=11.005, shapeName=dyn_strap_03_r_sphere, center=<6.0253 8.7147 -0.0022>}]
 dyn_strap_04_l [{shapeType=Sphere, radius=11.005, shapeName=dyn_strap_04_l_sphere, center=<-4.8724 -8.2318 -0.019>}]
 dyn_strap_04_r [{shapeType=Sphere, radius=11.005, shapeName=dyn_strap_04_r_sphere, center=<4.926 8.6793 -0.2644>}]
 dyn_strap_05_l [{shapeType=Sphere, radius=11.005, shapeName=dyn_strap_05_l_sphere, center=<-4.0414 -8.3183 -0.0221>}]
 dyn_strap_05_r [{shapeType=Sphere, radius=11.0242, shapeName=dyn_strap_05_r_sphere, center=<4.1167 8.7899 -1.3223>}]
 dyn_strap_06_l [{shapeType=Sphere, radius=10.1208, shapeName=dyn_strap_06_l_sphere, center=<-3.8203 -8.2692 0.0433>}]
 dyn_strap_07_l [{shapeType=Sphere, radius=11.005, shapeName=dyn_strap_07_l_sphere, center=<-4.949 -5.3551 0.4153>}]
 dyn_strap_base_l [{shapeType=Sphere, radius=5.305, shapeName=dyn_strap_base_l_sphere}]
 dyn_strap_base_r [{shapeType=Sphere, radius=5.305, shapeName=dyn_strap_base_r_sphere}]
 foot_l []
 foot_r []
 head [{shapeType=Sphere, radius=17.3125, center=<6.0031 -0.7188 0>}]
 lowerarm_l []
 lowerarm_r []
 neck_02 [{shapeType=Capsule, length=13.4375, radius=7.3175, center=<-0.8626 -2.7378 0>, rotation=<rot 90 180 -175.1686>}]
 pelvis [{
   shapeType=Capsule
   length=5.5683
   radius=19.5601
   shapeName=pelvis_capsule
   center=<-12.1082 -0.1717 0.1171>
   rotation=<rot 0 1.1886 -179.4437>
  }]
 root [{shapeType=Capsule, length=0.0001, radius=283.9999, shapeName=root_capsule, center=<0 47.5276 -267.1843>, rotation=<rot 0 -90 0>}]
 spine_01 [{shapeType=Capsule, length=10.6438, radius=14.9586, center=<-0.2732 -2.3371 0>}]
 spine_02 [
  {shapeType=Capsule, length=8.2264, radius=15.3322, center=<-1.711 -2.1647 0.0001>}
  {shapeType=Capsule, length=29.3956, radius=11.5807, center=<2.5024 0.4992 3.9825>, rotation=<rot 90 0.0049 0.0049>}
  {shapeType=Capsule, length=29.875, radius=11.1013, center=<2.4704 0.4991 -3.9825>, rotation=<rot 90 -0.0396 -0.0396>}
 ]
 spine_03 [{shapeType=Capsule, length=9.6107, radius=15.4031, center=<2.5296 -1.0693 0.0001>}]
 spine_04 [{shapeType=Capsule, length=12.8404, radius=16.1326, center=<2.766 -0.4734 0.0002>}]
 spine_05 [
  {shapeType=Capsule, length=16.2226, radius=16.8951, center=<-0.2772 0.9064 0>}
  {shapeType=Capsule, length=13.9518, radius=13.9425, center=<1.8873 0.4452 -0.0978>}
 ]
 thigh_l []
 thigh_r []
 upperarm_l []
 upperarm_r []
}
constraints [
 {
  bone1Name=spine_01
  bone2Name=pelvis
  swing1LimitDegrees=45
  swing1Motion=Free
  swing2LimitDegrees=45
  swing2Motion=Free
  twistLimitDegrees=45
  twistMotion=Free
 }
 {
  bone1Name=spine_02
  bone2Name=spine_01
  swing1LimitDegrees=45
  swing1Motion=Free
  swing2LimitDegrees=45
  swing2Motion=Free
  twistLimitDegrees=45
  twistMotion=Free
 }
 {
  bone1Name=spine_03
  bone2Name=spine_02
  swing1LimitDegrees=45
  swing1Motion=Free
  swing2LimitDegrees=45
  swing2Motion=Free
  twistLimitDegrees=45
  twistMotion=Free
 }
 {
  bone1Name=spine_04
  bone2Name=spine_03
  swing1LimitDegrees=45
  swing1Motion=Free
  swing2LimitDegrees=45
  swing2Motion=Free
  twistLimitDegrees=45
  twistMotion=Free
 }
 {
  bone1Name=spine_05
  bone2Name=spine_04
  swing1LimitDegrees=45
  swing1Motion=Free
  swing2LimitDegrees=45
  swing2Motion=Free
  twistLimitDegrees=45
  twistMotion=Free
 }
 {
  bone1Name=neck_02
  bone2Name=spine_05
  swing1LimitDegrees=45
  swing1Motion=Free
  swing2LimitDegrees=45
  swing2Motion=Free
  twistLimitDegrees=45
  twistMotion=Free
 }
 {
  bone1Name=head
  bone2Name=neck_02
  swing1LimitDegrees=45
  swing1Motion=Free
  swing2LimitDegrees=45
  swing2Motion=Free
  twistLimitDegrees=45
  twistMotion=Free
 }
 {
  bone1Name=clavicle_r
  bone2Name=spine_05
  swing1LimitDegrees=45
  swing1Motion=Free
  swing2LimitDegrees=45
  swing2Motion=Free
  twistLimitDegrees=45
  twistMotion=Free
 }
 {
  bone1Name=lowerarm_l
  bone2Name=upperarm_l
  swing1LimitDegrees=45
  swing1Motion=Free
  swing2LimitDegrees=45
  swing2Motion=Free
  twistLimitDegrees=45
  twistMotion=Free
 }
 {
  bone1Name=upperarm_l
  bone2Name=clavicle_l
  swing1LimitDegrees=45
  swing1Motion=Free
  swing2LimitDegrees=45
  swing2Motion=Free
  twistLimitDegrees=45
  twistMotion=Free
 }
 {
  bone1Name=calf_r
  bone2Name=thigh_r
  swing1LimitDegrees=45
  swing1Motion=Free
  swing2LimitDegrees=45
  swing2Motion=Free
  twistLimitDegrees=45
  twistMotion=Free
 }
 {
  bone1Name=lowerarm_r
  bone2Name=upperarm_r
  swing1LimitDegrees=45
  swing1Motion=Free
  swing2LimitDegrees=45
  swing2Motion=Free
  twistLimitDegrees=45
  twistMotion=Free
 }
 {
  bone1Name=upperarm_r
  bone2Name=clavicle_r
  swing1LimitDegrees=45
  swing1Motion=Free
  swing2LimitDegrees=45
  swing2Motion=Free
  twistLimitDegrees=45
  twistMotion=Free
 }
 {
  bone1Name=clavicle_l
  bone2Name=spine_05
  swing1LimitDegrees=45
  swing1Motion=Free
  swing2LimitDegrees=45
  swing2Motion=Free
  twistLimitDegrees=45
  twistMotion=Free
 }
 {bone1Name=dyn_strap_07_l, bone2Name=dyn_strap_06_l, swing1LimitDegrees=45, swing2LimitDegrees=45, twistLimitDegrees=10}
 {bone1Name=dyn_strap_01_l, bone2Name=dyn_strap_base_l, swing1LimitDegrees=45, swing2LimitDegrees=45, twistLimitDegrees=10}
 {bone1Name=dyn_strap_02_l, bone2Name=dyn_strap_01_l, swing1LimitDegrees=45, swing2LimitDegrees=45, twistLimitDegrees=10}
 {bone1Name=dyn_strap_03_l, bone2Name=dyn_strap_02_l, swing1LimitDegrees=45, swing2LimitDegrees=45, twistLimitDegrees=10}
 {bone1Name=dyn_strap_04_l, bone2Name=dyn_strap_03_l, swing1LimitDegrees=45, swing2LimitDegrees=45, twistLimitDegrees=10}
 {bone1Name=dyn_strap_05_l, bone2Name=dyn_strap_04_l, swing1LimitDegrees=45, swing2LimitDegrees=45, twistLimitDegrees=10}
 {bone1Name=dyn_strap_06_l, bone2Name=dyn_strap_05_l, swing1LimitDegrees=45, swing2LimitDegrees=45, twistLimitDegrees=10}
 {bone1Name=dyn_strap_03_r, bone2Name=dyn_strap_02_r, swing1LimitDegrees=45, swing2LimitDegrees=45, twistLimitDegrees=10}
 {bone1Name=dyn_strap_04_r, bone2Name=dyn_strap_03_r, swing1LimitDegrees=45, swing2LimitDegrees=45, twistLimitDegrees=10}
 {bone1Name=dyn_strap_05_r, bone2Name=dyn_strap_04_r, swing1LimitDegrees=45, swing2LimitDegrees=45, twistLimitDegrees=10}
 {
  bone1Name=dyn_strap_base_l
  bone2Name=pelvis
  swing1LimitDegrees=45
  swing1Motion=Free
  swing2LimitDegrees=45
  swing2Motion=Free
  twistLimitDegrees=45
  twistMotion=Free
 }
 {
  bone1Name=dyn_strap_base_r
  bone2Name=pelvis
  swing1LimitDegrees=45
  swing1Motion=Free
  swing2LimitDegrees=45
  swing2Motion=Free
  twistLimitDegrees=45
  twistMotion=Free
 }
 {
  bone1Name=thigh_l
  bone2Name=pelvis
  swing1LimitDegrees=45
  swing1Motion=Free
  swing2LimitDegrees=45
  swing2Motion=Free
  twistLimitDegrees=45
  twistMotion=Free
 }
 {
  bone1Name=foot_l
  bone2Name=calf_l
  swing1LimitDegrees=45
  swing1Motion=Free
  swing2LimitDegrees=45
  swing2Motion=Free
  twistLimitDegrees=45
  twistMotion=Free
 }
 {
  bone1Name=thigh_r
  bone2Name=pelvis
  swing1LimitDegrees=45
  swing1Motion=Free
  swing2LimitDegrees=45
  swing2Motion=Free
  twistLimitDegrees=45
  twistMotion=Free
 }
 {
  bone1Name=calf_l
  bone2Name=thigh_l
  swing1LimitDegrees=45
  swing1Motion=Free
  swing2LimitDegrees=45
  swing2Motion=Free
  twistLimitDegrees=45
  twistMotion=Free
 }
 {bone1Name=dyn_strap_02_r, bone2Name=dyn_strap_01_r, swing1LimitDegrees=45, swing2LimitDegrees=45, twistLimitDegrees=10}
 {
  bone1Name=foot_r
  bone2Name=calf_r
  swing1LimitDegrees=45
  swing1Motion=Free
  swing2LimitDegrees=45
  swing2Motion=Free
  twistLimitDegrees=45
  twistMotion=Free
 }
 {bone1Name=dyn_strap_01_r, bone2Name=dyn_strap_base_r, swing1LimitDegrees=45, swing2LimitDegrees=45, twistLimitDegrees=10}
 {
  bone1Name=pelvis
  bone2Name=root
  swing1LimitDegrees=45
  swing1Motion=Free
  swing2LimitDegrees=45
  swing2Motion=Free
  twistLimitDegrees=45
  twistMotion=Free
 }
]
